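# Niagara system for PistolCanary's directional tentacle attack: a short electric shockwave burst thrown forward along +X with debris.
# Staggered bursts drive ground-wave and outer-sphere meshes, impact and wave sprites, electric arcs and streaks, fanned dust, and glowing rocks.
# Curves animate dynamic material parameters and shift colors from hot magenta toward blue, while User.Scale sizes some elements.
# Rocks bounce under gravity, with collision switched off by a one-shot velocity and age check.
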NiagaraSystem /PistolCanary/Assets/FX/TentacleDirectionalAttack/TentacleDirectionalAttack/NS_PistolCanary_TentacleDirectionalAttack
dependencies {
 usedDataInterfaces [
  &NiagaraDataInterfaceCurve
  &NiagaraDataInterfaceVectorCurve
  &NiagaraDataInterfaceColorCurve
  &NiagaraDataInterfaceVector2DCurve
 ]
 usedDynamicInputs []
 usedModules [&SystemState]
}
emitters {
 Dust {
  emitterUpdateScript [
   &EmitterState(Life Cycle Mode=::NewEnumerator1, Loop Behavior=::NewEnumerator1, Loop Duration=1, UseLoopDelay=true)
   &SpawnBurst_Instantaneous(Spawn Count=11, Use Spawn Probability=false)
  ]
  particleSpawnScript [
   &FN_SetSpawnGroup(Normalize Spawn Group=true, Number of Spawn Groups=1, Particles Per Group=10)
   &InitializeParticle(
    Color Maximum=<rgba 0.8385 0.9347 1 0.4>
    Color Minimum=<rgba 0.8125 0.8227 1 0.3>
    Color Mode=::NewEnumerator2
    Lifetime Max=1.2
    Lifetime Min=0.75
    Lifetime Mode=::NewEnumerator1
    Sprite Size Mode=::NewEnumerator3
    Uniform Sprite Size=&Multiply_Float(
     A=&RandomRangeFloat(Minimum=250, Maximum=416.6667)
     B=&FloatFromCurve(
      FloatCurve=DataInterface() {propertyValues {Curve {keys [
          {interpMode=RCIM_Cubic, tangentMode=RCTM_User, time=0, value=0.5}
          {arriveTangentWeight=0.15, interpMode=RCIM_Cubic, leaveTangentWeight=0.0333, tangentMode=RCTM_User, time=0.5, value=1}
          {tangentMode=RCTM_User, time=1, value=0.5}
         ]}}}
      CurveIndex=$Particles.Normalized Spawn Group
      Scale Curve=$User.Scale
     )
    )
    UsePositionOffset=false
    Write Material Random=true
   ) {module=InitializeParticle001}
   &TorusLocation(
    Align Torus With Axis=true
    Handle Radius=5
    Large Radius=15
    MaskSpawnGroup=false
    Torus Distribution Mode:ENiagaraTorusDistributionMode=::NewEnumerator1
    U Position:NiagaraFloat=&Lerp_Float(A=-0.1, B=0.1, Alpha=$Particles.Normalized Spawn Group)
   )
   &AddVelocityFromPoint(
    Module.Use Falloff Distance=false
    Origin Offset=<7 0 0>
    Velocity Strength=&Multiply_Float(
     A=&RandomRangeFloat(Minimum=4000, Maximum=5000)
     B=&FloatFromCurve(
      FloatCurve=DataInterface() {propertyValues {Curve {keys [
          {arriveTangent=1.2775, interpMode=RCIM_Cubic, leaveTangent=1.2775, tangentMode=RCTM_User, time=0, value=0.4}
          {arriveTangent=1.5825, interpMode=RCIM_Cubic, leaveTangent=1.5825, tangentMode=RCTM_User, time=0.37, value=0.95}
          {arriveTangentWeight=0.045, interpMode=RCIM_Cubic, leaveTangentWeight=0.2027, tangentMode=RCTM_User, time=0.5, value=1.04}
          {arriveTangent=-1.4249, interpMode=RCIM_Cubic, leaveTangent=-1.4249, tangentMode=RCTM_User, time=0.63, value=0.95}
          {arriveTangent=-1.4935, interpMode=RCIM_Cubic, leaveTangent=-1.4935, tangentMode=RCTM_User, time=1, value=0.4}
         ]}}}
      CurveIndex=$Particles.Normalized Spawn Group
      Scale Curve=9
     )
    )
   )
  ]
  particleUpdateScript [
   &ParticleState()
   &Drag(Drag=&RandomRangeFloat(Minimum=4, Maximum=6), Use Linear Drag=true, Use Rotational Drag=false)
   &SpriteRotationRate(Rotation Rate=&Multiply_Float(
     A=&RandomRangeFloat(Minimum=-3, Maximum=3)
     B=&FloatFromCurve(
      FloatCurve=DataInterface() {propertyValues {Curve {keys [
          {arriveTangent=-4.0107, interpMode=RCIM_Cubic, leaveTangent=-4.0107, tangentMode=RCTM_User, time=-0.0046, value=1.0113}
          {arriveTangent=-0.2496, interpMode=RCIM_Cubic, leaveTangent=-0.2496, tangentMode=RCTM_User, time=0.5, value=0.16}
         ]}}}
      CurveIndex=$Emitter.Age
      Scale Curve=150
     )
    ))
   &WindForce(Air Resistance=0.4, Wind Speed=<0 0 1>, Wind Speed Scale=&RandomRangeFloat(Minimum=400, Maximum=900))
   &ScaleSpriteSize(Scale Factor=&Vector2DFromFloat(Value=&FloatFromCurve(FloatCurve=DataInterface() {propertyValues {Curve {keys [
          {arriveTangent=2.3678, interpMode=RCIM_Cubic, leaveTangent=2.3678, tangentMode=RCTM_User, time=0, value=0.2}
          {arriveTangent=0.7401, interpMode=RCIM_Cubic, leaveTangent=0.7401, tangentMode=RCTM_User, time=0.3, value=0.7}
          {arriveTangent=0.2539, interpMode=RCIM_Cubic, leaveTangent=0.2539, tangentMode=RCTM_User, time=0.9978, value=1}
         ]}}})))
   &SolveForcesAndVelocity(Clamp Velocity=false, Limit Acceleration=false)
   &ScaleColor(
    Scale Alpha=&FloatFromCurve(
     FloatCurve=DataInterface() {propertyValues {Curve {keys [
         {arriveTangent=5.4511, interpMode=RCIM_Cubic, leaveTangent=5.4511, tangentMode=RCTM_User, time=0, value=0.4}
         {arriveTangent=-1.4813, interpMode=RCIM_Cubic, leaveTangent=-1.4813, tangentMode=RCTM_User, time=0.2, value=1}
         {arriveTangent=-0.484, interpMode=RCIM_Cubic, leaveTangent=-0.484, tangentMode=RCTM_User, time=1, value=0}
        ]}}}
     Scale Curve=&RandomRangeFloat(Minimum=0.5, Maximum=0.3)
    )
    Scale RGB=&MakeVectorFromLinearColorRGB(Color=&ColorFromCurve(
      ColorCurve=DataInterface() {propertyValues {
        AlphaCurve {keys []}
        BlueCurve {keys [<at 0.0485 6 RCIM_Cubic>, {arriveTangent=-19.0486, interpMode=RCIM_Cubic, leaveTangent=-19.0486, time=0.1696, value=2}, <at 0.3295 0.6458>]}
        GreenCurve {keys [
          <at 0.0485 1.018 RCIM_Cubic>
          {arriveTangent=-2.2041, interpMode=RCIM_Cubic, leaveTangent=-2.2041, time=0.1696, value=0.4856}
          <at 0.3295 0.3985>
         ]}
        RedCurve {keys [
          <at 0.0485 5.153 RCIM_Cubic>
          {arriveTangent=-16.8034, interpMode=RCIM_Cubic, leaveTangent=-16.8034, time=0.1696, value=1.4424}
          <at 0.3295 0.4299>
         ]}
       }}
      CurveIndex=$Emitter.Age
     ))
    ScaleA=true
    ScaleRGB=true
   )
   &DynamicMaterialParameters(
    Index 0 Param 1=&Add_Float(
     A=$Particles.RandomFloat
     B=&FloatFromCurve(
      FloatCurve=DataInterface() {propertyValues {Curve {keys [
          {arriveTangent=2.7533, interpMode=RCIM_Cubic, leaveTangent=1.8958, tangentMode=RCTM_User, time=0, value=1}
          {arriveTangent=0.8292, interpMode=RCIM_Cubic, leaveTangent=0.8292, tangentMode=RCTM_User, time=0.321, value=1.3453}
          {arriveTangent=0.0529, interpMode=RCIM_Cubic, leaveTangent=0.2021, tangentMode=RCTM_User, time=1, value=1.5}
         ]}}}
      CurveIndex=&Add_Float(A=&Multiply_Float(A=$Particles.Initial.MaterialRandom, B=0.5), B=$Particles.NormalizedAge)
     )
    )
    Index 0 Param 2=&FloatFromCurve(FloatCurve=DataInterface() {propertyValues {Curve {keys [
         {arriveTangent=0.0453, interpMode=RCIM_Cubic, tangentMode=RCTM_Break, time=0, value=-1}
         {arriveTangent=0.0604, interpMode=RCIM_Cubic, leaveTangent=0.0604, tangentMode=RCTM_User, time=1, value=-1}
        ]}}})
    Index 0 Param 3=&FloatFromCurve(FloatCurve=DataInterface() {propertyValues {Curve {keys [
         {arriveTangent=-53.9199, interpMode=RCIM_Cubic, leaveTangent=-8.9727, tangentMode=RCTM_User, time=0, value=1}
         {arriveTangent=0.2347, interpMode=RCIM_Cubic, leaveTangent=0.2723, tangentMode=RCTM_Break, time=0.25, value=0}
        ]}}})
    Index 0 Param 4=&FloatFromCurve(FloatCurve=DataInterface() {propertyValues {Curve {keys [<at 0.5 1>,<at 1 0>]}}})
    Index 2 Param 1=&FloatFromCurve(FloatCurve=DataInterface() {propertyValues {Curve {keys [
         {interpMode=RCIM_Cubic, leaveTangent=4.2485, tangentMode=RCTM_Break, time=0, value=0}
         {arriveTangent=1.0548, interpMode=RCIM_Cubic, leaveTangent=1.0548, tangentMode=RCTM_User, time=0.17, value=0.5715}
         <at 1 1>
        ]}}})
    Index 2 Param 2=&FloatFromCurve(
     FloatCurve=DataInterface() {propertyValues {Curve {keys [
         {arriveTangent=0.9316, interpMode=RCIM_Cubic, leaveTangent=0.9316, tangentMode=RCTM_User, time=0, value=0.1}
         {arriveTangent=0.0259, interpMode=RCIM_Cubic, leaveTangent=0.0259, tangentMode=RCTM_User, time=1, value=0.4}
        ]}}}
     Scale Curve=&RandomRangeFloat(Minimum=0.2, Maximum=0.7)
    )
    Index 2 Param 3=&FloatFromCurve(FloatCurve=DataInterface() {propertyValues {Curve {keys [
         {interpMode=RCIM_Cubic, leaveTangent=2.3064, tangentMode=RCTM_Break, time=0, value=1}
         {interpMode=RCIM_Cubic, leaveTangent=11.1233, tangentMode=RCTM_Break, time=1, value=2}
        ]}}})
    Index 2 Param 4=&FloatFromCurve(
     FloatCurve=DataInterface() {propertyValues {Curve {keys [
         {arriveTangent=-0.6867, interpMode=RCIM_Cubic, leaveTangent=-0.6867, tangentMode=RCTM_User, time=0, value=1}
         {interpMode=RCIM_Cubic, leaveTangent=-1.2956, tangentMode=RCTM_Break, time=1, value=0.75}
        ]}}}
     Scale Curve=1.6
    )
    Module.Param0WriteEnabled=true
    Module.Param0WriteEnabled2=true
    Module.Param1WriteEnabled=false
    Module.Param1WriteEnabled2=true
    Module.Param2WriteEnabled=true
    Module.Param2WriteEnabled2=true
    Module.Param3WriteEnabled=true
    Module.Param3WriteEnabled2=true
    Write Parameter Index 2=true
   )
   &CameraOffset(Camera Offset Amount=/Niagara/DynamicInputs/Math/CalculateParticleRadius(
     Radius Calculation Type:ENiagaraCollisionRadiusOptions=::NewEnumerator0
     Particle Radius:NiagaraFloat=1
     Mesh Dimensions:Vector3f=<1 1 1>
     Method for Calculating Particle Radius:ENiagaraCalculateRadiusOptions=::NewEnumerator2
    ))
  ]
  rendererClasses [&NiagaraSpriteRendererProperties]
  renderers [
   {rendererClass=&NiagaraSpriteRendererProperties}
   {rendererClass=&NiagaraSpriteRendererProperties, rendererIndex=1}
  ]
 }
 ElectricArcs {
  emitterUpdateScript [
   &EmitterState(Life Cycle Mode=::NewEnumerator1, Loop Behavior=::NewEnumerator1, Loop Duration=0.25, UseLoopDelay=false)
   &SpawnBurst_Instantaneous(Spawn Count=2, Use Spawn Probability=false, UseLoopCountLimit=false)
   &SpawnBurst_Instantaneous(Spawn Count=2, Spawn Time=0.04, Use Spawn Probability=false, UseLoopCountLimit=false) {module=SpawnBurst_Instantaneous002}
   &SpawnBurst_Instantaneous(Spawn Count=3, Spawn Time=0.07, Use Spawn Probability=false, UseLoopCountLimit=false) {module=SpawnBurst_Instantaneous001}
  ]
  particleSpawnScript [
   /Niagara/Modules/Spawn/Initialization/InitializeParticle(
    Color=&RandomRangeLinearColor(Minimum=<rgba 1.9477 1.0986 4 1>, Maximum=<rgba 4.19 2.4994 6 1>)
    Lifetime=&UniformRangedFloat(Minimum=0.1, Maximum=0.15)
    Module.Write Color=true
    Module.Write Lifetime=true
    Module.Write Mass=true
    Module.Write Material Random=true
    Module.Write Position=false
    Module.Write Scale=false
    Module.Write SpriteRotation=true
    Module.Write SpriteSize=true
    Position=&SimulationPosition()
    Sprite Rotation=&UniformRangedFloat(Minimum=0, Maximum=360)
    Sprite Size=&Multiply_Vector2D(
     A=&UniformRangedVector2D(Minimum=<70 70>, Maximum=<150 150>)
     B=&Vector2DFromFloat(Value=$User.Scale)
    )
   )
   &DynamicMaterialParameters(
    Index 0 Param 1=$Particles.MaterialRandom
    Index 0 Param 2=&Lerp_Float(A=0.7, B=1.25, Alpha=$Particles.MaterialRandom)
    Index 0 Param 4=1
    Module.Param0WriteEnabled=true
    Module.Param1WriteEnabled=true
    Module.Param2WriteEnabled=false
    Module.Param3WriteEnabled=true
   )
   &CameraOffset(Camera Offset Amount=25)
   &ShapeLocation(
    Cylinder Height=200
    Cylinder Radius=1200
    Hemicircle X=true
    MaskSpawnGroup=false
    Override Seed=false
    Shape Primitive=::NewEnumerator1
    Surface Only=true
    Use Endcaps In Surface Only Mode=false
    UseOwnerScale=true
   )
   &AddVelocity(Module.Use Falloff Distance=false, Origin Offset=<-120 0 -20>, Override Seed=false, Velocity Mode=::NewEnumerator1, Velocity Speed=1800)
  ]
  particleUpdateScript [
   &ParticleState()
   &SpriteRotationRate(Rotation Rate=&RandomRangeFloat(Minimum=-200, Maximum=200))
   &DynamicMaterialParameters(
    Index 0 Param 3=&FloatFromCurve(FloatCurve=DataInterface() {propertyValues {Curve {keys [<at 0.5 0 RCIM_Cubic>, {arriveTangent=3.7979, interpMode=RCIM_Cubic, leaveTangent=3.7979, tangentMode=RCTM_User, time=1, value=1}]}}})
    Index 1 Param 2=0.7
    Index 1 Param 3=$System.RipplePan
    Index 1 Param 4=600
    Module.Param0WriteEnabled=false
    Module.Param0WriteEnabled1=true
    Module.Param1WriteEnabled=false
    Module.Param1WriteEnabled1=true
    Module.Param2WriteEnabled=true
    Module.Param2WriteEnabled1=true
    Module.Param3WriteEnabled=false
    Module.Param3WriteEnabled1=true
    Write Parameter Index 1=true
   ) {module=DynamicMaterialParameters001}
   &ScaleSpriteSize(Scale Factor=&Vector2DFromFloat(Value=&FloatFromCurve(FloatCurve=DataInterface() {propertyValues {Curve {keys [{arriveTangent=0.3147, interpMode=RCIM_Cubic, leaveTangent=0.3147, tangentMode=RCTM_User, time=0, value=0.8}, <at 1 1>]}}})))
   &Drag(Drag=3, Use Linear Drag=true, Use Rotational Drag=false)
   &SolveForcesAndVelocity(Clamp Velocity=false, Limit Acceleration=false)
   &ScaleColor(
    Scale RGB=&MakeVectorFromLinearColorRGB(Color=&ColorFromCurve(
      ColorCurve=DataInterface() {propertyValues {
        AlphaCurve {keys []}
        BlueCurve {keys [
          <at 0.0485 1.8072 RCIM_Cubic>
          {arriveTangent=-2.3921, interpMode=RCIM_Cubic, leaveTangent=-2.3921, time=0.2983, value=1.6918}
          <at 0.534 0.6458>
         ]}
        GreenCurve {keys [
          <at 0.0485 0.4164 RCIM_Cubic>
          {arriveTangent=0.1445, interpMode=RCIM_Cubic, leaveTangent=0.1445, time=0.2983, value=0.3165}
          <at 0.534 0.4866>
         ]}
        RedCurve {keys [<at 0.0485 4 RCIM_Cubic>, {arriveTangent=-7.6897, interpMode=RCIM_Cubic, leaveTangent=-7.6897, time=0.2983, value=2}, <at 0.534 0.2666>]}
       }}
      CurveIndex=$Emitter.Age
     ))
    ScaleA=false
    ScaleRGB=true
   ) {module=ScaleColor001}
  ]
  rendererClasses [&NiagaraSpriteRendererProperties]
  renderers [{rendererClass=&NiagaraSpriteRendererProperties}]
 }
 ElectricStreaks {
  emitterUpdateScript [
   &EmitterState(Life Cycle Mode=::NewEnumerator1, Loop Behavior=::NewEnumerator1, Loop Duration=0.17, UseLoopDelay=false)
   &SpawnBurst_Instantaneous(Spawn Count=8, Use Spawn Probability=false, UseLoopCountLimit=false)
   &SpawnBurst_Instantaneous(Spawn Count=3, Spawn Group=1, Spawn Time=0.06, Use Spawn Probability=false, UseLoopCountLimit=false) {module=SpawnBurst_Instantaneous001}
  ]
  particleSpawnScript [
   &InitializeParticle(
    Color=&RandomRangeLinearColor(Minimum=<rgba 0.0315 0.2104 1 1>, Maximum=<rgba 2 0.2237 1.2847 1>)
    Lifetime=&UniformRangedFloat(Minimum=0.55, Maximum=0.65)
    Module.Write Color=true
    Module.Write Lifetime=true
    Module.Write Mass=true
    Module.Write Material Random=true
    Module.Write Position=false
    Module.Write Scale=false
    Module.Write SpriteRotation=false
    Module.Write SpriteSize=true
    Position=&SimulationPosition()
    Sprite Size=&MakeVector2D(
     X=&Multiply_Float(A=&RandomRangeFloat(Minimum=70, Maximum=90), B=1)
     Y=&FloatFromCurve(
      FloatCurve=DataInterface() {propertyValues {Curve {keys [
          {interpMode=RCIM_Cubic, tangentMode=RCTM_User, time=0, value=0.4}
          {interpMode=RCIM_Cubic, tangentMode=RCTM_User, time=0.5, value=1}
          {tangentMode=RCTM_User, time=1, value=0.4}
         ]}}}
      CurveIndex=&ReturnNormalizedExecIndex()
      Scale Curve=&RandomRangeFloat(Minimum=450, Maximum=700)
     )
    )
   )
   &DynamicMaterialParameters(
    Index 0 Param 1=&RandomRangeFloat(Minimum=0, Maximum=1)
    Index 0 Param 2=&Lerp_Float(A=0.7, B=0.125, Alpha=$Particles.MaterialRandom)
    Module.Param0WriteEnabled=true
    Module.Param1WriteEnabled=true
    Module.Param2WriteEnabled=false
    Module.Param3WriteEnabled=false
   )
   &CameraOffset(Camera Offset Amount=25)
   &AddVelocity(
    Cone Angle=140
    Cone Axis=<0 0 1>
    Module.UseConeFalloff=true
    Module.UseVelDistribution=false
    Override Seed=false
    Speed Falloff From Cone Axis=-1
    Velocity Mode=::NewEnumerator2
    Velocity Speed=100
   ) {enabled=false}
   &SpriteFacingAndAlignment(
    Sprite Alignment=&RotateVectorByQuaternion(
     Vector To Rotate=&RandomRangeVector(Minimum=<0.8 0 0.5>, Maximum=<1 0 0>)
     Quaternion=&MakeQuatFromAxisAngle(
      Axis=<0 0 1>
      AngleInDegrees=&Add_Float(
       A=&Lerp_Float(A=-55, B=55, Alpha=&ReturnNormalizedExecIndex())
       B=&RandomRangeFloat(Minimum=-15, Maximum=15)
      )
      NormalizeAxis=true
     )
    )
    WriteAlignment=true
    WriteFacing=false
   )
  ]
  particleUpdateScript [
   &ParticleState()
   &SolveForcesAndVelocity(Clamp Velocity=false, Limit Acceleration=false)
   &ScaleSpriteSize(
    Non-Uniform Curve Scale=&Vector2DFromFloat(Value=15)
    Non-Uniform Curve Sprite Scale=DataInterface() {propertyValues {
      XCurve {keys [
        {arriveTangent=0.8718, interpMode=RCIM_Cubic, leaveTangent=0.8718, tangentMode=RCTM_User, time=0, value=0.8}
        {arriveTangentWeight=0.1795, interpMode=RCIM_Cubic, leaveTangentWeight=0.1795, tangentMode=RCTM_User, time=0.5, value=1}
        {arriveTangent=-3.9093, interpMode=RCIM_Cubic, leaveTangent=-3.9093, tangentMode=RCTM_User, time=1, value=0.2}
       ]}
      YCurve {keys [
        {arriveTangent=2.2429, interpMode=RCIM_Cubic, leaveTangent=2.2429, tangentMode=RCTM_User, time=0, value=0.5}
        {arriveTangent=0.3913, arriveTangentWeight=0.1667, interpMode=RCIM_Cubic, leaveTangent=0.3913, tangentMode=RCTM_User, time=0.5, value=1}
        {arriveTangent=0.1824, interpMode=RCIM_Cubic, leaveTangent=0.1824, tangentMode=RCTM_User, time=1, value=1.12}
       ]}
     }}
    Scale Sprite Size Mode=::NewEnumerator3
   )
   &DynamicMaterialParameters(
    Index 0 Param 3=&FloatFromCurve(FloatCurve=DataInterface() {propertyValues {Curve {keys [<at 0.5 0 RCIM_Cubic>, {arriveTangent=3.8263, interpMode=RCIM_Cubic, leaveTangent=3.8263, tangentMode=RCTM_User, time=1, value=1}]}}})
    Index 0 Param 4=&Divide_Float(
     A=&CalculateParticleRadius(
      Radius Calculation Type:ENiagaraCollisionRadiusOptions=::NewEnumerator0
      Particle Radius:NiagaraFloat=1
      Mesh Dimensions:Vector3f=<1 1 1>
      Method for Calculating Particle Radius:ENiagaraCalculateRadiusOptions=::NewEnumerator2
     )
     B=&RandomRangeFloat(Minimum=4000, Maximum=10000)
    )
    Index 1 Param 2=$System.RippleMult
    Index 1 Param 3=$System.RipplePan
    Index 1 Param 4=800
    Module.Param0WriteEnabled=false
    Module.Param0WriteEnabled1=true
    Module.Param1WriteEnabled=false
    Module.Param1WriteEnabled1=true
    Module.Param2WriteEnabled=true
    Module.Param2WriteEnabled1=true
    Module.Param3WriteEnabled=true
    Module.Param3WriteEnabled1=true
    Write Parameter Index 1=true
   ) {module=DynamicMaterialParameters001}
   &ScaleColor(
    Scale Alpha=&FloatFromCurve(FloatCurve=DataInterface() {propertyValues {Curve {keys [<at 0.5 1 RCIM_Cubic>,<at 1 0>]}}})
    Scale RGB=&MakeVectorFromLinearColorRGB(Color=&ColorFromCurve(
      ColorCurve=DataInterface() {propertyValues {
        AlphaCurve {keys []}
        BlueCurve {keys [
          <at 0.0485 1.8072 RCIM_Cubic>
          {arriveTangent=-2.3921, interpMode=RCIM_Cubic, leaveTangent=-2.3921, time=0.2983, value=1.6918}
          <at 0.534 0.6458>
         ]}
        GreenCurve {keys [
          <at 0.0485 0.4164 RCIM_Cubic>
          {arriveTangent=0.1445, interpMode=RCIM_Cubic, leaveTangent=0.1445, time=0.2983, value=0.3165}
          <at 0.534 0.4866>
         ]}
        RedCurve {keys [<at 0.0485 4 RCIM_Cubic>, {arriveTangent=-7.6897, interpMode=RCIM_Cubic, leaveTangent=-7.6897, time=0.2983, value=2}, <at 0.534 0.2666>]}
       }}
      CurveIndex=$Emitter.Age
     ))
    ScaleA=true
    ScaleRGB=true
   )
  ]
  rendererClasses [&NiagaraSpriteRendererProperties]
  renderers [{rendererClass=&NiagaraSpriteRendererProperties}]
 }
 GroundWave {
  emitterUpdateScript [
   &EmitterState()
   &SpawnBurst_Instantaneous(Use Spawn Probability=false, UseLoopCountLimit=false)
   &SpawnBurst_Instantaneous(Spawn Time=0.08, Use Spawn Probability=false, UseLoopCountLimit=false) {module=SpawnBurst_Instantaneous001}
   &SpawnBurst_Instantaneous(Spawn Time=0.16, Use Spawn Probability=false, UseLoopCountLimit=false) {module=SpawnBurst_Instantaneous002}
  ]
  particleSpawnScript [
   /Niagara/Modules/Spawn/Initialization/V2/InitializeParticle(
    Color=<rgba 1.5 1.5 1.5 1>
    Lifetime=0.4
    Mesh Scale=&Multiply_Vector(A=<3.5 4.25 5>, B=&VectorFromFloat(Value=$User.Scale))
    Mesh Scale Mode=::NewEnumerator4
    UsePositionOffset=false
    Write Material Random=true
   )
   &InitialMeshOrientation(Mesh Orientation Mode=::NewEnumerator2, Use Rotation Vector=false)
   &CameraOffset(Camera Offset Amount=40)
  ]
  particleUpdateScript [
   &ParticleState()
   /SetVariables(Particles.Position=&MakeVector(
     X=&FloatFromCurve(
      FloatCurve=DataInterface() {propertyValues {Curve {keys [
          {arriveTangent=1.9657, interpMode=RCIM_Cubic, leaveTangent=1.9657, tangentMode=RCTM_User, time=0, value=-0.2}
          {arriveTangent=0.1617, interpMode=RCIM_Cubic, leaveTangent=0.1617, tangentMode=RCTM_User, time=1, value=1}
         ]}}}
      Scale Curve=300
     )
     Y=0
     Z=0
    )) {isSetParameters=true}
   &ScaleMeshSize(Scale Factor=&VectorFromCurve(VectorCurve=DataInterface() {propertyValues {
       XCurve {keys [
         {arriveTangent=1.4912, interpMode=RCIM_Cubic, leaveTangent=1.4912, tangentMode=RCTM_User, time=0, value=0.2}
         {arriveTangent=0.1443, interpMode=RCIM_Cubic, leaveTangent=0.1443, tangentMode=RCTM_User, time=1, value=1}
        ]}
       YCurve {keys [
         {arriveTangent=1.4912, interpMode=RCIM_Cubic, leaveTangent=1.4912, tangentMode=RCTM_User, time=0, value=0.2}
         {arriveTangent=0.1443, interpMode=RCIM_Cubic, leaveTangent=0.1443, tangentMode=RCTM_User, time=1, value=1}
        ]}
       ZCurve {keys [
         {arriveTangent=5.9564, interpMode=RCIM_Cubic, leaveTangent=5.9564, tangentMode=RCTM_User, time=0, value=0.5}
         {arriveTangent=-0.1, interpMode=RCIM_Cubic, leaveTangent=-0.1, time=0.2, value=1}
         {arriveTangent=-1.1924, interpMode=RCIM_Cubic, leaveTangent=-1.1924, tangentMode=RCTM_User, time=1, value=0.4}
        ]}
      }})) {enabled=false}
   &ScaleMeshSize(Scale Factor=&VectorFromCurve(VectorCurve=DataInterface() {propertyValues {
       XCurve {keys [
         {arriveTangent=1.7518, interpMode=RCIM_Cubic, leaveTangent=1.7518, tangentMode=RCTM_User, time=0, value=0.2}
         {arriveTangent=0.3714, interpMode=RCIM_Cubic, leaveTangent=0.3714, tangentMode=RCTM_User, time=1, value=1}
        ]}
       YCurve {keys [
         {arriveTangent=1.2753, interpMode=RCIM_Cubic, leaveTangent=1.2753, tangentMode=RCTM_User, time=0, value=0.3}
         {arriveTangent=0.3714, interpMode=RCIM_Cubic, leaveTangent=0.3714, tangentMode=RCTM_User, time=1, value=1}
        ]}
       ZCurve {keys [
         {arriveTangent=1.5507, interpMode=RCIM_Cubic, leaveTangent=1.5507, tangentMode=RCTM_User, time=0, value=0.16}
         {arriveTangent=-0.502, interpMode=RCIM_Cubic, leaveTangent=-0.502, tangentMode=RCTM_User, time=0.4, value=0.6}
         <at 1 0.2>
        ]}
      }})) {module=ScaleMeshSize001}
   &DynamicMaterialParameters(
    Index 0 Param 1=&FloatFromCurve(FloatCurve=DataInterface() {propertyValues {Curve {keys [
         {arriveTangent=-4.9069, interpMode=RCIM_Cubic, time=0, value=1}
         {arriveTangent=-4.5698, interpMode=RCIM_Cubic, leaveTangent=-4.5698, tangentMode=RCTM_User, time=0.4, value=0.3}
         <at 1 -1 RCIM_Cubic>
        ]}}})
    Index 0 Param 2=&FloatFromCurve(FloatCurve=DataInterface() {propertyValues {Curve {keys [{arriveTangent=-3.2248, interpMode=RCIM_Cubic, time=0, value=1.5}, {interpMode=RCIM_Cubic, leaveTangent=-3.8057, time=1, value=0}]}}})
    Index 0 Param 4=&FloatFromCurve(FloatCurve=DataInterface() {propertyValues {Curve {keys [{arriveTangent=2.0289, interpMode=RCIM_Cubic, leaveTangent=2.0289, tangentMode=RCTM_User, time=0, value=0}, <at 1 1 RCIM_Cubic>]}}})
    Module.Param0WriteEnabled=true
    Module.Param1WriteEnabled=true
    Module.Param2WriteEnabled=true
    Module.Param3WriteEnabled=true
   )
   &ScaleColor(
    Scale Alpha=&FloatFromCurve(FloatCurve=DataInterface() {propertyValues {Curve {keys [
         {arriveTangent=26.1682, interpMode=RCIM_Cubic, leaveTangent=26.1682, tangentMode=RCTM_User, time=0, value=0}
         {interpMode=RCIM_Cubic, tangentMode=RCTM_User, time=0.1, value=1}
         {tangentMode=RCTM_User, time=1, value=0}
        ]}}})
    Scale RGB=&Lerp_Vector(
     A=<3 1 1.5>
     B=<0.5 1 1.8>
     Alpha=&FloatFromCurve(
      FloatCurve=DataInterface() {propertyValues {Curve {keys [{interpMode=RCIM_Cubic, tangentMode=RCTM_User, time=0.05, value=0}, {interpMode=RCIM_Cubic, leaveTangent=2.3979, time=0.4, value=1}]}}}
      CurveIndex=&Lerp_Float(A=$Emitter.Age, B=$Particles.Age, Alpha=0.7)
     )
    )
    ScaleA=true
    ScaleRGB=true
   )
  ]
  rendererClasses [&NiagaraMeshRendererProperties]
  renderers [{rendererClass=&NiagaraMeshRendererProperties, rendererIndex=0}]
 }
 ImpactPoint {
  emitterUpdateScript [
   &EmitterState()
   &SpawnBurst_Instantaneous(Use Spawn Probability=false, UseLoopCountLimit=false)
   &SpawnBurst_Instantaneous(Spawn Time=0.1, Use Spawn Probability=false, UseLoopCountLimit=false) {module=SpawnBurst_Instantaneous001}
   &SpawnBurst_Instantaneous(Spawn Time=0.2, Use Spawn Probability=false, UseLoopCountLimit=false) {module=SpawnBurst_Instantaneous002}
   /SetVariables(Emitter.TxtureuV:Vector4f=&MakeVector4(X=1, Y=-0.5, Z=0, W=0.5)) {isSetParameters=true, module=SetVariables_#1}
  ]
  particleSpawnScript [
   /SetVariables(Particles.ColorEnd:LinearColor=<rgba 0.0088 0.0256 0.5104 1>, Particles.ColorStart:LinearColor=<rgba 14.9884 1.4917 50 1>) {isSetParameters=true, module=SetVariables_#2}
   &InitializeParticle(
    Lifetime=0.5
    Position Offset=<10 0 -20>
    Sprite Size Mode=::NewEnumerator3
    Uniform Sprite Size=&Lerp_Float(
     A=400
     B=100
     Alpha=&FloatFromCurve(
      FloatCurve=DataInterface() {propertyValues {Curve {keys [
          {arriveTangent=8.5679, interpMode=RCIM_Cubic, leaveTangent=8.5679, tangentMode=RCTM_User, time=0.05, value=0}
          {arriveTangentWeight=0.8986, interpMode=RCIM_Cubic, tangentMode=RCTM_User, time=0.2, value=1}
         ]}}}
      CurveIndex=$Emitter.NormalizedLoopAge
     )
    )
    UsePositionOffset=true
    Write Material Random=true
   )
   &CameraOffset(Camera Offset Amount=50)
   &SpriteFacingAndAlignment(Sprite Facing=<0 0 1>, WriteAlignment=false, WriteFacing=true)
  ]
  particleUpdateScript [
   &ParticleState()
   &Timeline(
    Initial Start Time:NiagaraFloat=&RandomRangeFloat(Minimum=0, Maximum=50)
    Play Rate=&Add_Float(
     A=&FloatFromCurve(
      FloatCurve=DataInterface() {propertyValues {Curve {keys [
          {interpMode=RCIM_Cubic, tangentMode=RCTM_Break, time=0, value=1}
          {arriveTangent=-2.9724, interpMode=RCIM_Cubic, tangentMode=RCTM_Break, time=1.1, value=0}
         ]}}}
      CurveIndex=$Particles.Age
      Scale Curve=0.75
     )
     B=0.5
    )
    Timeline End=100000
   )
   &ScaleSpriteSize(
    Uniform Curve Scale=30
    Uniform Curve Sprite Scale=DataInterface() {propertyValues {Curve {keys [
        {arriveTangent=0.9259, interpMode=RCIM_Cubic, leaveTangent=0.9259, tangentMode=RCTM_User, time=0, value=0.5}
        {arriveTangent=0.8647, interpMode=RCIM_Cubic, leaveTangent=0.8647, tangentMode=RCTM_User, time=1, value=1.5}
       ]}}}
   )
   &DynamicMaterialParameters(
    Index 0 Param 1=&FloatFromCurve(FloatCurve=DataInterface() {propertyValues {Curve {keys [{interpMode=RCIM_Cubic, leaveTangent=3.9935, tangentMode=RCTM_Break, time=0, value=0}, {tangentMode=RCTM_User, time=0.5, value=1}]}}})
    Index 0 Param 3=$Particles.Timeline.Time
    Index 1 Param 1=50
    Module.Param0WriteEnabled=true
    Module.Param0WriteEnabled1=true
    Module.Param1WriteEnabled=true
    Module.Param1WriteEnabled1=true
    Module.Param2WriteEnabled=true
    Module.Param2WriteEnabled1=true
    Module.Param3WriteEnabled=true
    Module.Param3WriteEnabled1=true
    Write Parameter Index 1=true
   )
   &ScaleColor(
    Scale Alpha=&FloatFromCurve(FloatCurve=DataInterface() {propertyValues {Curve {keys [
         {arriveTangent=0.2514, interpMode=RCIM_Cubic, leaveTangent=0.2514, tangentMode=RCTM_User, time=0, value=1}
         {arriveTangent=-3.0342, interpMode=RCIM_Cubic, leaveTangent=-3.0342, tangentMode=RCTM_User, time=0.1408, value=0.9847}
         {arriveTangent=-0.0405, interpMode=RCIM_Cubic, leaveTangent=-0.23, tangentMode=RCTM_Break, time=1, value=0}
        ]}}})
    Scale RGB=&VectorFromFloat(Value=&Add_Float(
      A=&FloatFromCurve(
       FloatCurve=DataInterface() {propertyValues {Curve {keys [
           {arriveTangent=9.0725, time=0, value=0}
           {arriveTangent=100, leaveTangent=100, time=0.0649, value=0.1991}
           {arriveTangent=-22.974, leaveTangent=-22.974, time=0.2319, value=1.9348}
           <at 0.295 1.7737>
           <at 0.5 0>
           {arriveTangent=-0.0405, interpMode=RCIM_Cubic, leaveTangent=-0.0405, tangentMode=RCTM_User, time=1, value=0}
          ]}}}
       CurveIndex=$Particles.Age
      )
      B=1
     ))
    ScaleA=true
    ScaleRGB=false
   )
   &ScaleColor(
    Scale RGB=&MakeVectorFromLinearColorRGB(Color=&LerpLinearColors(
      StartColor=$Particles.ColorStart
      EndColor=$Particles.ColorEnd
      LerpFactor=&FloatFromCurve(FloatCurve=DataInterface() {propertyValues {Curve {keys [
           {interpMode=RCIM_Cubic, tangentMode=RCTM_User, time=0, value=0}
           {arriveTangent=1.7408, interpMode=RCIM_Cubic, leaveTangent=1.7408, tangentMode=RCTM_User, time=0.2, value=0.8}
           {interpMode=RCIM_Cubic, tangentMode=RCTM_User, time=0.5, value=1}
          ]}}})
     ))
    ScaleA=false
    ScaleRGB=true
   ) {module=ScaleColor001}
  ]
  rendererClasses [&NiagaraSpriteRendererProperties]
  renderers [{rendererClass=&NiagaraSpriteRendererProperties}]
 }
 OuterSphere {
  emitterUpdateScript [
   &EmitterState()
   &SpawnBurst_Instantaneous(Use Spawn Probability=false, UseLoopCountLimit=false)
   &SpawnBurst_Instantaneous(Spawn Time=0.08, Use Spawn Probability=false, UseLoopCountLimit=false) {module=SpawnBurst_Instantaneous001}
   &SpawnBurst_Instantaneous(Spawn Time=0.16, Use Spawn Probability=false, UseLoopCountLimit=false) {module=SpawnBurst_Instantaneous002}
  ]
  particleSpawnScript [
   &InitializeParticle(
    Color=<rgba 1.5 1.5 1.5 1>
    Lifetime=0.3
    Mesh Scale=&Multiply_Vector(A=<3 3 3.2>, B=&VectorFromFloat(Value=10))
    Mesh Scale Mode=::NewEnumerator4
    UsePositionOffset=false
    Write Material Random=true
   )
   /SetVariables(Particles.LightPosition:NiagaraPosition=<0 0 40>) {isSetParameters=true}
   &InitialMeshOrientation(
    Mesh Axis To Orient=<0 0 1>
    Mesh Orientation Mode=::NewEnumerator2
    Rotation=&RandomRangeVector(Minimum=<0 0 0>, Maximum=<0 0 1>)
    Use Rotation Vector=true
   )
  ]
  particleUpdateScript [
   &ParticleState()
   /SetVariables(Particles.Position=&MakeVector(
     X=&FloatFromCurve(
      FloatCurve=DataInterface() {propertyValues {Curve {keys [
          {arriveTangent=1.4387, interpMode=RCIM_Cubic, leaveTangent=1.4387, tangentMode=RCTM_User, time=0, value=0}
          {arriveTangent=0.3784, interpMode=RCIM_Cubic, leaveTangent=0.3784, tangentMode=RCTM_User, time=1, value=1}
         ]}}}
      Scale Curve=2450
     )
     Y=0
     Z=&FloatFromCurve(
      FloatCurve=DataInterface() {propertyValues {Curve {keys [
          {arriveTangent=1.4387, time=0, value=0.1}
          {arriveTangent=0.3784, interpMode=RCIM_Cubic, leaveTangent=0.3784, tangentMode=RCTM_User, time=1, value=1}
         ]}}}
      Scale Curve=500
     )
    )) {isSetParameters=true}
   /SetVariables(Particles.LightColor=&ColorFromCurve(ColorCurve=DataInterface() {propertyValues {
       AlphaCurve {keys [
         <at 0 1 RCIM_Cubic>
         {arriveTangent=-2.4738, interpMode=RCIM_Cubic, leaveTangent=-2.4738, tangentMode=RCTM_User, time=0.3627, value=0.7008}
         <at 1 0 RCIM_Cubic>
        ]}
       BlueCurve {keys [<at 0.1733 36.9992 RCIM_Cubic>,<at 0.647 25 RCIM_Cubic>]}
       GreenCurve {keys [<at 0.1733 1.9736 RCIM_Cubic>,<at 0.647 10.9288 RCIM_Cubic>]}
       RedCurve {keys [<at 0.1733 40 RCIM_Cubic>,<at 0.647 4.126 RCIM_Cubic>]}
      }})) {isSetParameters=true}
   &ScaleMeshSize(Scale Factor=&VectorFromFloat(Value=&FloatFromCurve(FloatCurve=DataInterface() {propertyValues {Curve {keys [
          {arriveTangent=1.8818, interpMode=RCIM_Cubic, leaveTangent=1.8818, tangentMode=RCTM_User, time=0, value=0.1}
          {arriveTangent=0.3784, interpMode=RCIM_Cubic, leaveTangent=0.3784, tangentMode=RCTM_User, time=1, value=1}
         ]}}})))
   &DynamicMaterialParameters(
    Index 0 Param 1=&FloatFromCurve(
     FloatCurve=DataInterface() {propertyValues {Curve {keys [
         {arriveTangent=-2.9934, interpMode=RCIM_Cubic, leaveTangent=-2.9934, tangentMode=RCTM_User, time=0, value=1}
         {interpMode=RCIM_Cubic, leaveTangent=-1.2287, time=1, value=-0.25}
        ]}}}
     Scale Curve=2
    )
    Index 0 Param 2=1
    Index 0 Param 3=&FloatFromCurve(
     FloatCurve=DataInterface() {propertyValues {Curve {keys [
         {arriveTangent=2.2611, interpMode=RCIM_Cubic, leaveTangent=2.2611, tangentMode=RCTM_User, time=0, value=0}
         {arriveTangent=0.402, interpMode=RCIM_Cubic, leaveTangent=0.402, tangentMode=RCTM_User, time=1, value=1}
        ]}}}
     Scale Curve=5
    )
    Index 0 Param 4=&FloatFromCurve(FloatCurve=DataInterface() {propertyValues {Curve {keys [{arriveTangent=-2.6143, interpMode=RCIM_Cubic, time=0, value=1}, <at 0.7 0>]}}})
    Module.Param0WriteEnabled=true
    Module.Param1WriteEnabled=true
    Module.Param2WriteEnabled=true
    Module.Param3WriteEnabled=true
   )
   &ScaleColor(
    Scale Alpha=&FloatFromCurve(FloatCurve=DataInterface() {propertyValues {Curve {keys [
         {arriveTangent=26.1682, interpMode=RCIM_Cubic, leaveTangent=26.1682, tangentMode=RCTM_User, time=0, value=0}
         {arriveTangent=-2.2763, interpMode=RCIM_Cubic, leaveTangent=-2.2763, tangentMode=RCTM_User, time=0.07, value=1}
         {tangentMode=RCTM_User, time=1, value=0}
        ]}}})
    Scale RGB=&Lerp_Vector(
     A=<1.2 0.2 0.8>
     B=<0.2 1 1>
     Alpha=&FloatFromCurve(
      FloatCurve=DataInterface() {propertyValues {Curve {keys [{interpMode=RCIM_Cubic, tangentMode=RCTM_User, time=0.1, value=0}, {interpMode=RCIM_Cubic, tangentMode=RCTM_User, time=0.4, value=1}]}}}
      CurveIndex=&Lerp_Float(A=$Emitter.Age, B=$Particles.Age, Alpha=0.6)
     )
    )
    ScaleA=true
    ScaleRGB=true
   )
  ]
  rendererClasses [&NiagaraMeshRendererProperties]
  renderers [{rendererClass=&NiagaraMeshRendererProperties, rendererIndex=0}]
 }
 Rocks {
  emitterUpdateScript [
   &EmitterState(Enable Distance Culling=true, MaxDistance=3000, Module.bUseMaxDistance=true, Module.bUseMinDistance=false, Scalability Mode=::NewEnumerator1)
   &SpawnBurst_Instantaneous(Spawn Count=20, Spawn Time=0.05, Use Spawn Probability=false, UseLoopCountLimit=false)
  ]
  particleSpawnScript [
   &InitializeParticle(
    Color Maximum=<rgba 0.5 0.5 0.5 1>
    Color Minimum=<rgba 0.25 0.25 0.25 1>
    Color Mode=::NewEnumerator2
    Lifetime=2
    Mesh Scale Mode=::NewEnumerator1
    Mesh Uniform Scale Max=0.32
    Mesh Uniform Scale Min=0.12
    UsePositionOffset=false
    Write Material Random=true
   )
   &InitialMeshOrientation(
    Rotation=&RandomRangeVector(Minimum=<0 0 0>, Maximum=<1 1 1>)
    Use Orientation Vector=true
    Use Rotation Vector=true
   )
   &ShapeLocation(
    MaskSpawnGroup=false
    Non Uniform Scale=<1 1 0.25>
    Offset=<0 0 15>
    Offset Mode=::NewEnumerator0
    Override Seed=false
    Sphere Radius=25
    UseOwnerScale=true
   )
   &AddVelocity(
    Module.Use Falloff Distance=false
    Origin Offset=<-25 0 -15>
    Override Seed=false
    Velocity Mode=::NewEnumerator1
    Velocity Speed=&RandomRangeFloat(Minimum=700, Maximum=1600)
   )
   /SetVariables(Particles.SpriteSize=&Vector2DFromFloat(Value=&Multiply_Float(A=&MakeFloatFromVector(Vector=$Particles.Scale, Channel=::NewEnumerator0), B=150))) {isSetParameters=true}
   &DynamicMaterialParameters(
    Index 3 Param 1=4
    Index 3 Param 2=0.2
    Module.Param0WriteEnabled=false
    Module.Param0WriteEnabled3=true
    Module.Param1WriteEnabled=false
    Module.Param1WriteEnabled3=true
    Module.Param2WriteEnabled=false
    Module.Param2WriteEnabled3=true
    Module.Param3WriteEnabled=false
    Module.Param3WriteEnabled3=true
    Write Parameter Index 3=true
   )
  ]
  particleUpdateScript [
   &ParticleState()
   &ScaleSpriteSize(Uniform Curve Sprite Scale=DataInterface() {propertyValues {Curve {keys [
        <at 0 1 RCIM_Cubic>
        {arriveTangent=-0.5, interpMode=RCIM_Cubic, leaveTangent=-0.5, time=0.1, value=1.25}
        {arriveTangent=-1.3889, interpMode=RCIM_Cubic, leaveTangent=-1.3889, time=0.2, value=0.9}
        {interpMode=RCIM_Cubic, tangentMode=RCTM_User, time=1, value=0}
       ]}}})
   &GravityForce(Gravity=<0 0 -1600>)
   &DoOnce(Trigger Condition:NiagaraBool=&Boolean_AndOperation(
     Bool1=&SetBoolByFloatComparison(A=&VectorLength(Vector=$Particles.Velocity), B=150, Comparison Type=::NewEnumerator6)
     Bool2=&SetBoolByFloatComparison(A=$Particles.Age, B=0.5)
    ))
   &Collision(
    Advanced Aging Rate=1
    Age Colliding Particles=false
    CPU Collision Trace Channel=ECC_Visibility
    Collision Enabled=&Boolean_NotOperation(Bool=$Particles.DoOnce.TriggeredOnce)
    Friction=0.5
    Friction During a Bounce=0.3
    Module.Control Rest State Via Surface Normal=false
    Module.EnableMaxCollisionCount=false
    Module.Randomize Collision Normal Vector=true
    Particle Radius Scale=2
    Radius Calculation Type=::NewEnumerator1
    Randomize Collision Normal=0.25
    Restitution=0.4
   )
   &MeshRotationForce(Rotation=&RandomRangeVector(Minimum=<-10 -10 -10>, Maximum=<10 10 10>))
   &SolveForcesAndVelocity(Clamp Velocity=false, Limit Acceleration=false)
   &ScaleMeshSize(Scale Factor=&VectorFromFloat(Value=&FloatFromCurve(
      FloatCurve=DataInterface() {propertyValues {Curve {keys [
          {arriveTangentWeight=0.2357, interpMode=RCIM_Cubic, leaveTangentWeight=0.2357, tangentMode=RCTM_User, time=0.5, value=1}
          {arriveTangent=-4.4912, interpMode=RCIM_Cubic, leaveTangent=-4.4912, tangentMode=RCTM_User, time=1, value=0}
         ]}}}
      Scale Curve=20
     )))
   /SetVariables(Particles.SpriteColor=&MakeLinearColorFromVectorAndFloat(
     "Vector (RGB)"=&MakeVectorFromLinearColorRGB(Color=<rgba 2.636 0.377 3 0.5>)
     "Float (Alpha)"=&FloatFromCurve(FloatCurve=DataInterface() {propertyValues {Curve {keys [
          <at 0 1 RCIM_Cubic>
          {arriveTangent=-2.8631, interpMode=RCIM_Cubic, leaveTangent=-2.8631, tangentMode=RCTM_User, time=0.1155, value=0.9}
          <at 0.2748 0.3227>
          <at 1 0>
         ]}}})
    )) {isSetParameters=true}
  ]
  rendererClasses [&NiagaraMeshRendererProperties, &NiagaraSpriteRendererProperties]
  renderers [
   {rendererClass=&NiagaraMeshRendererProperties, rendererIndex=0}
   {rendererClass=&NiagaraSpriteRendererProperties, rendererIndex=1}
  ]
 }
 Wave {
  emitterUpdateScript [
   &EmitterState()
   &SpawnBurst_Instantaneous(Use Spawn Probability=false, UseLoopCountLimit=false)
   /SetVariables(Emitter.TxtureuV:Vector4f=&MakeVector4(X=1, Y=-0.5, Z=0, W=0.5)) {isSetParameters=true, module=SetVariables_#1}
  ]
  particleSpawnScript [
   /SetVariables(Particles.ColorEnd:LinearColor=<rgba 0.0088 0.0256 0.5104 1>, Particles.ColorStart:LinearColor=<rgba 14.9884 1.4917 50 1>) {isSetParameters=true, module=SetVariables_#2}
   &InitializeParticle(
    Lifetime=0.5
    Position Offset=<20 0 -15>
    Sprite Size Mode=::NewEnumerator3
    Uniform Sprite Size=&Lerp_Float(
     A=500
     B=30
     Alpha=&FloatFromCurve(
      FloatCurve=DataInterface() {propertyValues {Curve {keys [
          {arriveTangent=8.5679, interpMode=RCIM_Cubic, leaveTangent=8.5679, tangentMode=RCTM_User, time=0.05, value=0}
          {arriveTangentWeight=0.8986, interpMode=RCIM_Cubic, tangentMode=RCTM_User, time=0.2, value=1}
         ]}}}
      CurveIndex=$Emitter.NormalizedLoopAge
     )
    )
    UsePositionOffset=true
    Write Material Random=true
   )
   &CameraOffset(Camera Offset Amount=50)
   &SpriteFacingAndAlignment(Sprite Facing=<0 0 1>, WriteAlignment=false, WriteFacing=true)
  ]
  particleUpdateScript [
   &ParticleState()
   &Timeline(
    Initial Start Time:NiagaraFloat=&RandomRangeFloat(Minimum=0, Maximum=50)
    Play Rate=&Add_Float(
     A=&FloatFromCurve(
      FloatCurve=DataInterface() {propertyValues {Curve {keys [
          {interpMode=RCIM_Cubic, tangentMode=RCTM_Break, time=0, value=1}
          {arriveTangent=-2.9724, interpMode=RCIM_Cubic, tangentMode=RCTM_Break, time=1.1, value=0}
         ]}}}
      CurveIndex=$Particles.Age
      Scale Curve=0.75
     )
     B=0.5
    )
    Timeline End=100000
   )
   &ScaleSpriteSize(
    Uniform Curve Scale=15
    Uniform Curve Sprite Scale=DataInterface() {propertyValues {Curve {keys [
        {arriveTangent=0.9259, interpMode=RCIM_Cubic, leaveTangent=0.9259, tangentMode=RCTM_User, time=0, value=0.5}
        {arriveTangent=0.8647, interpMode=RCIM_Cubic, leaveTangent=0.8647, tangentMode=RCTM_User, time=1, value=1.5}
       ]}}}
   )
   &DynamicMaterialParameters(
    Index 0 Param 1=&FloatFromCurve(FloatCurve=DataInterface() {propertyValues {Curve {keys [{interpMode=RCIM_Cubic, leaveTangent=3.9935, tangentMode=RCTM_Break, time=0, value=0}, {tangentMode=RCTM_User, time=0.5, value=1}]}}})
    Index 0 Param 3=$Particles.Timeline.Time
    Module.Param0WriteEnabled=true
    Module.Param1WriteEnabled=true
    Module.Param2WriteEnabled=true
    Module.Param3WriteEnabled=true
   )
   &ScaleColor(
    Scale Alpha=&FloatFromCurve(FloatCurve=DataInterface() {propertyValues {Curve {keys [
         {arriveTangent=0.2514, interpMode=RCIM_Cubic, leaveTangent=0.2514, tangentMode=RCTM_User, time=0, value=1}
         {arriveTangent=-3.0342, interpMode=RCIM_Cubic, leaveTangent=-3.0342, tangentMode=RCTM_User, time=0.1408, value=0.9847}
         {arriveTangent=-0.0405, interpMode=RCIM_Cubic, leaveTangent=-0.23, tangentMode=RCTM_Break, time=1, value=0}
        ]}}})
    Scale RGB=&VectorFromFloat(Value=&Add_Float(
      A=&FloatFromCurve(
       FloatCurve=DataInterface() {propertyValues {Curve {keys [
           {arriveTangent=9.0725, time=0, value=0}
           {arriveTangent=100, leaveTangent=100, time=0.0649, value=0.1991}
           {arriveTangent=-22.974, leaveTangent=-22.974, time=0.2319, value=1.9348}
           <at 0.295 1.7737>
           <at 0.5 0>
           {arriveTangent=-0.0405, interpMode=RCIM_Cubic, leaveTangent=-0.0405, tangentMode=RCTM_User, time=1, value=0}
          ]}}}
       CurveIndex=$Particles.Age
      )
      B=1
     ))
    ScaleA=true
    ScaleRGB=false
   )
   &ScaleColor(
    Scale RGB=&MakeVectorFromLinearColorRGB(Color=&LerpLinearColors(
      StartColor=$Particles.ColorStart
      EndColor=$Particles.ColorEnd
      LerpFactor=&FloatFromCurve(FloatCurve=DataInterface() {propertyValues {Curve {keys [
           {interpMode=RCIM_Cubic, tangentMode=RCTM_User, time=0, value=0}
           {arriveTangent=1.7408, interpMode=RCIM_Cubic, leaveTangent=1.7408, tangentMode=RCTM_User, time=0.2, value=0.8}
           {interpMode=RCIM_Cubic, tangentMode=RCTM_User, time=0.5, value=1}
          ]}}})
     ))
    ScaleA=false
    ScaleRGB=true
   ) {module=ScaleColor001}
  ]
  rendererClasses [&NiagaraSpriteRendererProperties]
  renderers [{rendererClass=&NiagaraSpriteRendererProperties}]
 }
 Wind {
  emitterUpdateScript [
   &EmitterState(Enable Distance Culling=true, MaxDistance=6000, Module.bUseMaxDistance=true, Module.bUseMinDistance=false, Scalability Mode=::NewEnumerator1)
   &SpawnRate(
    SpawnRate=&FloatFromCurve(
     FloatCurve=DataInterface() {propertyValues {Curve {keys [
         {arriveTangent=77.1005, interpMode=RCIM_Cubic, leaveTangent=77.1005, tangentMode=RCTM_User, time=0.05, value=0}
         <at 0.07 1 RCIM_Cubic>
         {arriveTangent=-14.3902, interpMode=RCIM_Cubic, leaveTangent=-14.3902, tangentMode=RCTM_User, time=0.25, value=0}
        ]}}}
     CurveIndex=$System.Age
     Scale Curve=40
    )
    Use Spawn Probability=false
   )
   &SpawnBurst_Instantaneous(Spawn Count=4, Use Spawn Probability=false, UseLoopCountLimit=false)
  ]
  particleSpawnScript [
   &InitializeParticle(
    Color=<rgba 1 1 1 0.35>
    Lifetime=&UniformRangedFloat(Minimum=0.32, Maximum=0.5)
    Mesh Scale=&Lerp_Vector(A=<3 3 4>, B=<6 6 6>, Alpha=&RandomRangeFloat(Minimum=0, Maximum=1))
    Mesh Scale Mode=::NewEnumerator4
    UsePositionOffset=false
    Write Material Random=true
   ) {module=InitializeParticle001}
   &InitialMeshOrientation(
    Mesh Axis To Orient=<0 0 1>
    Mesh Orientation Mode=::NewEnumerator2
    Rotation=&RandomRangeVector(Minimum=<0 0 0>, Maximum=<0 0 1>)
    Use Rotation Vector=true
   )
  ]
  particleUpdateScript [
   &ParticleState()
   &UpdateMeshOrientation(Rotation Rate=&RandomRangeFloat(Minimum=-1, Maximum=-2.5), Rotation Vector=<0 0 1>)
   &ScaleMeshSize(Scale Factor=&VectorFromCurve(VectorCurve=DataInterface() {propertyValues {
       XCurve {keys [
         {arriveTangent=1.9177, interpMode=RCIM_Cubic, leaveTangent=1.9177, tangentMode=RCTM_User, time=0, value=0.1}
         {arriveTangent=0.2322, interpMode=RCIM_Cubic, leaveTangent=0.2322, tangentMode=RCTM_User, time=0.8, value=1}
         {arriveTangent=-3.7145, interpMode=RCIM_Cubic, leaveTangent=-3.7145, tangentMode=RCTM_User, time=1, value=0.7}
        ]}
       YCurve {keys [
         {arriveTangent=1.9177, interpMode=RCIM_Cubic, leaveTangent=1.9177, tangentMode=RCTM_User, time=0, value=0.1}
         {arriveTangent=0.2322, interpMode=RCIM_Cubic, leaveTangent=0.2322, tangentMode=RCTM_User, time=0.8, value=1}
         {arriveTangent=-3.7145, interpMode=RCIM_Cubic, leaveTangent=-3.7145, tangentMode=RCTM_User, time=1, value=0.7}
        ]}
       ZCurve {keys [
         {arriveTangent=3.0696, interpMode=RCIM_Cubic, leaveTangent=3.0696, tangentMode=RCTM_User, time=0, value=0.5}
         {arriveTangent=-0.3, interpMode=RCIM_Cubic, leaveTangent=-0.3, time=0.35, value=0.9}
         {interpMode=RCIM_Cubic, leaveTangent=-1.6862, time=1, value=0.2}
        ]}
      }}))
   /SetVariables(Particles.Position=&AddVectorToPosition(
     Position=&Multiply_VectorByFloat(
      Vector=&ReturnMeshOrientationAxis(Mesh Orientation Axis=::ZAxis, Mesh Orientation Quat=$Particles.MeshOrientation)
      Float=&FloatFromCurve(
       FloatCurve=DataInterface() {propertyValues {Curve {keys [{arriveTangent=6.1711, interpMode=RCIM_Cubic, leaveTangent=6.1711, tangentMode=RCTM_User, time=0, value=-0.15}, <at 0.45 1 RCIM_Cubic>]}}}
       CurveIndex=$Particles.Age
       Scale Curve=&RandomRangeFloat(Minimum=500, Maximum=700)
      )
     )
     Vector=&MakeVector(
      X=0
      Y=0
      Z=&FloatFromCurve(
       FloatCurve=DataInterface() {propertyValues {Curve {keys [
           {arriveTangent=1.4387, time=0, value=0.2}
           {arriveTangent=0.3784, interpMode=RCIM_Cubic, leaveTangent=0.3784, tangentMode=RCTM_User, time=1, value=1}
          ]}}}
       Scale Curve=200
      )
     )
    )) {isSetParameters=true}
   &CameraOffset(Camera Offset Amount=&Lerp_Float(A=70, B=140, Alpha=$Particles.NormalizedAge))
   &ScaleColor(
    Scale Alpha=&FloatFromCurve(FloatCurve=DataInterface() {propertyValues {Curve {keys [
         {arriveTangent=8.125, interpMode=RCIM_Cubic, time=0, value=0}
         {interpMode=RCIM_Cubic, tangentMode=RCTM_User, time=0.3, value=1}
         {arriveTangent=-1.557, interpMode=RCIM_Cubic, leaveTangent=-1.557, tangentMode=RCTM_User, time=0.7791, value=0.7}
         {interpMode=RCIM_Cubic, leaveTangent=-2.3395, time=1, value=0}
        ]}}})
    Scale RGB=&MakeVectorFromLinearColorRGB(Color=&ColorFromCurve(
      ColorCurve=DataInterface() {propertyValues {
        AlphaCurve {keys []}
        BlueCurve {keys [<at 0.0485 6 RCIM_Cubic>, {arriveTangent=-19.0486, interpMode=RCIM_Cubic, leaveTangent=-19.0486, time=0.1696, value=2}, <at 0.3295 0.6458>]}
        GreenCurve {keys [
          <at 0.0485 1.018 RCIM_Cubic>
          {arriveTangent=-2.2041, interpMode=RCIM_Cubic, leaveTangent=-2.2041, time=0.1696, value=0.4856}
          <at 0.3295 0.3985>
         ]}
        RedCurve {keys [
          <at 0.0485 5.153 RCIM_Cubic>
          {arriveTangent=-16.8034, interpMode=RCIM_Cubic, leaveTangent=-16.8034, time=0.1696, value=1.4424}
          <at 0.3295 0.4299>
         ]}
       }}
      CurveIndex=$Emitter.Age
     ))
    ScaleA=true
    ScaleRGB=true
   )
   &DynamicMaterialParameters(
    Index 0 Param 1=&FloatFromCurve(FloatCurve=DataInterface() {propertyValues {Curve {keys [
         {arriveTangent=-13.4039, interpMode=RCIM_Cubic, leaveTangent=-13.4039, tangentMode=RCTM_User, time=0, value=1}
         <at 0.1 0>
         {arriveTangent=1.1111, interpMode=RCIM_Cubic, leaveTangent=1.1111, time=0.4, value=0}
         {interpMode=RCIM_Cubic, leaveTangent=16.4602, time=1, value=1}
        ]}}})
    Index 0 Param 2=&FloatFromCurve(
     FloatCurve=DataInterface() {propertyValues {Curve {keys [
         {arriveTangent=2.4843, interpMode=RCIM_Cubic, leaveTangent=2.4843, tangentMode=RCTM_User, time=0, value=0}
         {interpMode=RCIM_Cubic, leaveTangent=-0.8461, time=1, value=1}
        ]}}}
     Scale Curve=0.5
    )
    Index 0 Param 4=&FloatFromCurve(
     FloatCurve=DataInterface() {propertyValues {Curve {keys [
         {arriveTangent=0.2805, interpMode=RCIM_Cubic, time=0, value=1}
         {arriveTangent=2.4494, interpMode=RCIM_Cubic, leaveTangent=2.4494, tangentMode=RCTM_User, time=1, value=2}
        ]}}}
     Scale Curve=&RandomRangeFloat(Minimum=0.6, Maximum=0.8)
    )
    Module.Param0WriteEnabled=true
    Module.Param1WriteEnabled=true
    Module.Param2WriteEnabled=false
    Module.Param3WriteEnabled=true
   )
  ]
  rendererClasses [&NiagaraMeshRendererProperties]
  renderers [{rendererClass=&NiagaraMeshRendererProperties, rendererIndex=0}]
 }
 Wind001 {
  emitterUpdateScript [
   &EmitterState(Enable Distance Culling=true, MaxDistance=6000, Module.bUseMaxDistance=true, Module.bUseMinDistance=false, Scalability Mode=::NewEnumerator1)
   &SpawnBurst_Instantaneous(Use Spawn Probability=false, UseLoopCountLimit=false)
   &SpawnBurst_Instantaneous(Spawn Time=0.04, Use Spawn Probability=false, UseLoopCountLimit=false) {module=SpawnBurst_Instantaneous002}
   &SpawnBurst_Instantaneous(Spawn Time=0.075, Use Spawn Probability=false, UseLoopCountLimit=false) {module=SpawnBurst_Instantaneous001}
  ]
  particleSpawnScript [
   /SetVariables(Particles.FlipSide:NiagaraBool=&RandomBool()) {isSetParameters=true}
   &InitializeParticle(
    Color=<rgba 2 2 2 0.5>
    Lifetime=&UniformRangedFloat(Minimum=0.14, Maximum=0.16)
    Mesh Scale=&Lerp_Vector(A=<2.7 2.7 1.5>, B=<3 3 3>, Alpha=&RandomRangeFloat(Minimum=0, Maximum=20))
    Mesh Scale Mode=::NewEnumerator4
    UsePositionOffset=false
    Write Material Random=true
   ) {module=InitializeParticle001}
   &InitialMeshOrientation(
    Mesh Axis To Orient=&MakeCustomVectorFromBool(Bool=$Particles.FlipSide, True Vector=<0 0 1>, False Vector=<0 0 -1>)
    Mesh Orientation Mode=::NewEnumerator2
    Rotation=&RandomRangeVector(Minimum=<0 0 -0.3>, Maximum=<0 0 -0.27>)
    Use Rotation Vector=true
   )
  ]
  particleUpdateScript [
   &ParticleState()
   &UpdateMeshOrientation(
    Rotation Rate=&Multiply_Float(
     A=&MakeCustomFloatFromBool(Bool=$Particles.FlipSide, True Float=-1, False Float=1)
     B=&RandomRangeFloat(Minimum=3.5, Maximum=3.8)
    )
    Rotation Vector=<0 0 1>
   )
   /SetVariables(Particles.Position=&MakeVector(
     X=-15
     Y=&FloatFromCurve(
      FloatCurve=DataInterface() {propertyValues {Curve {keys [
          {arriveTangent=3.2328, interpMode=RCIM_Cubic, leaveTangent=3.2328, tangentMode=RCTM_User, time=0, value=-1}
          {interpMode=RCIM_Cubic, leaveTangent=1.2424, time=1, value=0.4}
         ]}}}
      CurveIndex=$System.NormalizedLoopAge
      Scale Curve=120
     )
     Z=&FloatFromCurve(
      FloatCurve=DataInterface() {propertyValues {Curve {keys [
          {arriveTangent=-1.8682, time=0, value=1}
          {arriveTangent=-2.0886, interpMode=RCIM_Cubic, leaveTangent=-2.0886, tangentMode=RCTM_User, time=1, value=0}
         ]}}}
      CurveIndex=$System.NormalizedLoopAge
      Scale Curve=200
     )
    )) {enabled=false, isSetParameters=true}
   &ScaleMeshSize(Scale Factor=&VectorFromCurve(
     VectorCurve=DataInterface() {propertyValues {
       XCurve {keys [
         {arriveTangent=1.1533, interpMode=RCIM_Cubic, leaveTangent=1.1533, tangentMode=RCTM_User, time=0, value=0.3}
         {arriveTangent=0.2322, interpMode=RCIM_Cubic, leaveTangent=0.2322, tangentMode=RCTM_User, time=1, value=1}
        ]}
       YCurve {keys [
         {arriveTangent=1.1533, interpMode=RCIM_Cubic, leaveTangent=1.1533, tangentMode=RCTM_User, time=0, value=0.3}
         {arriveTangent=0.2322, interpMode=RCIM_Cubic, leaveTangent=0.2322, tangentMode=RCTM_User, time=1, value=1}
        ]}
       ZCurve {keys [
         {arriveTangent=2.4044, interpMode=RCIM_Cubic, time=0, value=0.1}
         {arriveTangent=0.8008, interpMode=RCIM_Cubic, leaveTangent=0.8008, tangentMode=RCTM_User, time=1, value=1}
        ]}
      }}
     CurveIndex=$System.NormalizedLoopAge
    )) {enabled=false}
   &ScaleColor(
    Scale Alpha=&FloatFromCurve(FloatCurve=DataInterface() {propertyValues {Curve {keys [
         {arriveTangent=8.125, interpMode=RCIM_Cubic, time=0, value=0}
         {arriveTangentWeight=0.4805, interpMode=RCIM_Cubic, leaveTangentWeight=0.1363, tangentMode=RCTM_User, time=0.8, value=1}
         {interpMode=RCIM_Cubic, leaveTangent=-2.3395, time=1, value=0}
        ]}}})
    Scale RGB=&MakeVectorFromLinearColorRGB(Color=&ColorFromCurve(ColorCurve=DataInterface() {propertyValues {
        AlphaCurve {keys []}
        BlueCurve {keys [
          {arriveTangent=-19.0486, interpMode=RCIM_Cubic, time=0.3389, value=1}
          {arriveTangent=-0.5357, interpMode=RCIM_Cubic, leaveTangent=-0.5357, time=0.8297, value=12}
          {arriveTangent=-155.1807, interpMode=RCIM_Cubic, leaveTangent=-155.1807, tangentMode=RCTM_User, time=1, value=0.6458}
         ]}
        GreenCurve {keys [
          {arriveTangent=-2.2041, interpMode=RCIM_Cubic, time=0.3389, value=0.2185}
          {arriveTangent=8.5346, interpMode=RCIM_Cubic, leaveTangent=8.5346, tangentMode=RCTM_User, time=0.8297, value=2.036}
          {arriveTangent=-44.728, interpMode=RCIM_Cubic, leaveTangent=-44.728, tangentMode=RCTM_User, time=1, value=0.3985}
         ]}
        RedCurve {keys [
          {arriveTangent=-16.8034, interpMode=RCIM_Cubic, time=0.3389, value=0.3345}
          {arriveTangent=0.1443, interpMode=RCIM_Cubic, leaveTangent=0.1443, time=0.8297, value=10.3059}
          {arriveTangent=-155.1807, interpMode=RCIM_Cubic, leaveTangent=-155.1807, tangentMode=RCTM_User, time=1, value=0.4299}
         ]}
       }}))
    ScaleA=true
    ScaleRGB=true
   )
   &DynamicMaterialParameters(
    Index 0 Param 1=&FloatFromCurve(FloatCurve=DataInterface() {propertyValues {Curve {keys [
         {arriveTangent=-5.84, interpMode=RCIM_Cubic, time=0, value=1}
         <at 0.6 0 RCIM_Cubic>
         {interpMode=RCIM_Cubic, leaveTangent=16.4602, time=1, value=1}
        ]}}})
    Index 0 Param 2=&FloatFromCurve(FloatCurve=DataInterface() {propertyValues {Curve {keys [
         {arriveTangent=2.4843, interpMode=RCIM_Cubic, leaveTangent=2.4843, tangentMode=RCTM_User, time=0, value=0}
         {interpMode=RCIM_Cubic, leaveTangent=-0.8461, time=1, value=1}
        ]}}})
    Index 0 Param 4=&FloatFromCurve(
     FloatCurve=DataInterface() {propertyValues {Curve {keys [
         {arriveTangent=0.2805, interpMode=RCIM_Cubic, time=0, value=1}
         {arriveTangent=2.4494, interpMode=RCIM_Cubic, leaveTangent=2.4494, tangentMode=RCTM_User, time=1, value=2}
        ]}}}
     Scale Curve=&RandomRangeFloat(Minimum=0.6, Maximum=0.8)
    )
    Module.Param0WriteEnabled=true
    Module.Param1WriteEnabled=true
    Module.Param2WriteEnabled=false
    Module.Param3WriteEnabled=true
   )
  ]
  rendererClasses [&NiagaraMeshRendererProperties]
  renderers [{rendererClass=&NiagaraMeshRendererProperties, rendererIndex=0}]
 }
 WindCone {
  emitterUpdateScript [
   &EmitterState()
   &SpawnBurst_Instantaneous(Use Spawn Probability=false, UseLoopCountLimit=false)
  ]
  particleSpawnScript [
   &InitializeParticle(
    Color=<rgba 8.5464 0.1729 10 0.5>
    Lifetime=0.3
    Mesh Scale=&Multiply_Vector(A=<2 1.4 1.4>, B=&VectorFromFloat(Value=$User.Scale))
    Mesh Scale Mode=::NewEnumerator4
    Sprite Size Mode=::NewEnumerator3
    Uniform Sprite Size=250
    UsePositionOffset=false
    Write Material Random=true
   ) {module=InitializeParticle001}
   &InitialMeshOrientation(Mesh Orientation Mode=::NewEnumerator2, Orientation Vector=<-1 0 -0.1>, Use Rotation Vector=false)
   &DynamicMaterialParameters(
    Index 0 Param 1=0.7
    Index 0 Param 2=0.8
    Index 0 Param 3=2
    Index 2 Param 1=10
    Index 3 Param 1=0.2
    Index 3 Param 2=3
    Index 3 Param 3=1
    Index 3 Param 4=1
    Module.Param0WriteEnabled=true
    Module.Param0WriteEnabled2=true
    Module.Param0WriteEnabled3=true
    Module.Param1WriteEnabled=true
    Module.Param1WriteEnabled2=false
    Module.Param1WriteEnabled3=true
    Module.Param2WriteEnabled=true
    Module.Param2WriteEnabled2=false
    Module.Param2WriteEnabled3=true
    Module.Param3WriteEnabled=false
    Module.Param3WriteEnabled2=false
    Module.Param3WriteEnabled3=true
    Write Parameter Index 2=true
    Write Parameter Index 3=true
   )
   &CameraOffset(Camera Offset Amount=35)
  ]
  particleUpdateScript [
   &ParticleState()
   &ScaleMeshSize(Scale Factor=&VectorFromCurve(VectorCurve=DataInterface() {propertyValues {
       XCurve {keys [
         {arriveTangent=1.5385, time=0, value=0}
         {arriveTangent=0.3483, interpMode=RCIM_Cubic, leaveTangent=0.3483, tangentMode=RCTM_User, time=1, value=1}
        ]}
       YCurve {keys [
         {arriveTangent=2.3109, interpMode=RCIM_Cubic, leaveTangent=2.3109, tangentMode=RCTM_User, time=0, value=0}
         {arriveTangent=0.6124, interpMode=RCIM_Cubic, leaveTangent=0.6124, tangentMode=RCTM_User, time=1, value=1}
        ]}
       ZCurve {keys [
         {arriveTangent=2.3109, interpMode=RCIM_Cubic, leaveTangent=2.3109, tangentMode=RCTM_User, time=0, value=0}
         {arriveTangent=0.6124, interpMode=RCIM_Cubic, leaveTangent=0.6124, tangentMode=RCTM_User, time=1, value=1}
        ]}
      }}))
   /SetVariables(Particles.LightPosition:NiagaraPosition=&MakeVector(
     X=&FloatFromCurve(
      FloatCurve=DataInterface() {propertyValues {Curve {keys [
          {arriveTangent=2.5106, interpMode=RCIM_Cubic, leaveTangent=2.5106, tangentMode=RCTM_User, time=0, value=-0.1}
          {arriveTangent=0.6738, interpMode=RCIM_Cubic, leaveTangent=0.6738, tangentMode=RCTM_User, time=1, value=1}
         ]}}}
      Scale Curve=400
     )
     Y=0
     Z=35
    )) {isSetParameters=true}
   &ScaleColor(
    Scale Alpha=&FloatFromCurve(FloatCurve=DataInterface() {propertyValues {Curve {keys [
         {arriveTangent=48.0273, interpMode=RCIM_Cubic, leaveTangent=48.0273, tangentMode=RCTM_User, time=0, value=0}
         <at 0.05 1 RCIM_Cubic>
         {interpMode=RCIM_Cubic, leaveTangent=-3.1778, time=1, value=0}
        ]}}})
    Scale RGB=&MakeVectorFromLinearColorRGB(Color=&ColorFromCurve(ColorCurve=DataInterface() {propertyValues {
        AlphaCurve {keys []}
        BlueCurve {keys [<at 0.1026 1 RCIM_Cubic>,<at 0.959 0.5 RCIM_Cubic>]}
        GreenCurve {keys [<at 0.1026 1 RCIM_Cubic>,<at 0.959 0.2461 RCIM_Cubic>]}
        RedCurve {keys [<at 0.1026 1 RCIM_Cubic>,<at 0.959 0.2651 RCIM_Cubic>]}
       }}))
    ScaleA=true
    ScaleRGB=true
   )
   /SetVariables(Particles.FlareColor=&MultiplyLinearColor(Linear Color=$Particles.Color, Scale Factor=<2 2 2 3>)) {isSetParameters=true}
   /SetVariables(Particles.LightColor=&ColorFromCurve(ColorCurve=DataInterface() {propertyValues {
       AlphaCurve {keys [
         <at 0 1 RCIM_Cubic>
         {arriveTangent=-2.4738, interpMode=RCIM_Cubic, leaveTangent=-2.4738, tangentMode=RCTM_User, time=0.3627, value=0.7008}
         <at 1 0 RCIM_Cubic>
        ]}
       BlueCurve {keys [<at 0.1733 36.9992 RCIM_Cubic>,<at 0.647 25 RCIM_Cubic>]}
       GreenCurve {keys [<at 0.1733 1.9736 RCIM_Cubic>,<at 0.647 10.9288 RCIM_Cubic>]}
       RedCurve {keys [<at 0.1733 40 RCIM_Cubic>,<at 0.647 4.126 RCIM_Cubic>]}
      }})) {isSetParameters=true}
   &DynamicMaterialParameters(
    Index 1 Param 1=&FloatFromCurve(FloatCurve=DataInterface() {propertyValues {Curve {keys [<at 0 -1 RCIM_Cubic>,<at 1 0>]}}})
    Module.Param0WriteEnabled1=true
    Module.Param1WriteEnabled1=true
    Module.Param2WriteEnabled1=false
    Module.Param3WriteEnabled1=false
    Write Parameter Index 0=false
    Write Parameter Index 1=true
   ) {module=DynamicMaterialParameters001}
  ]
  rendererClasses [
   &NiagaraMeshRendererProperties
   &NiagaraSpriteRendererProperties
   &NiagaraLightRendererProperties
  ]
  renderers [
   {rendererClass=&NiagaraMeshRendererProperties, rendererIndex=0}
   {rendererClass=&NiagaraSpriteRendererProperties, rendererIndex=1}
   {rendererClass=&NiagaraLightRendererProperties, rendererIndex=2}
  ]
 }
}
summary {
 systemName=NS_PistolCanary_TentacleDirectionalAttack
 emitters [
  {emitterName=GroundWave, rendererClasses [&NiagaraMeshRendererProperties]}
  {emitterName=OuterSphere, rendererClasses [&NiagaraMeshRendererProperties]}
  {
   emitterName=WindCone
   rendererClasses [
    &NiagaraMeshRendererProperties
    &NiagaraSpriteRendererProperties
    &NiagaraLightRendererProperties
   ]
  }
  {emitterName=Dust, rendererClasses [&NiagaraSpriteRendererProperties]}
  {emitterName=Wind, rendererClasses [&NiagaraMeshRendererProperties]}
  {emitterName=ElectricArcs, rendererClasses [&NiagaraSpriteRendererProperties]}
  {emitterName=ElectricStreaks, rendererClasses [&NiagaraSpriteRendererProperties]}
  {emitterName=Rocks, rendererClasses [&NiagaraMeshRendererProperties, &NiagaraSpriteRendererProperties]}
  {emitterName=Wave, rendererClasses [&NiagaraSpriteRendererProperties]}
  {emitterName=ImpactPoint, rendererClasses [&NiagaraSpriteRendererProperties]}
  {emitterName=Wind001, rendererClasses [&NiagaraMeshRendererProperties]}
 ]
 userVariables [
  User.Magnitude {type=NiagaraFloat, defaultValue=0, description=""}
  User.Scale {type=NiagaraFloat, defaultValue=20, description=""}
 ]
}
